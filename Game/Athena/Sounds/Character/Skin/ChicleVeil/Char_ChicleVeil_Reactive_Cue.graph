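# Sound cue for the reactive character sound of the ChicleVeil skin, played as player foley.
# A short fixed delay leads into a random node that picks one of three equally weighted wave variations.
# It uses the basic foley attenuation and the player foley sound class, with a raised volume multiplier.
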SoundCue /Game/Athena/Sounds/Character/Skin/ChicleVeil/Char_ChicleVeil_Reactive_Cue
root=Delay_0
nodes [
 {id=Delay_0, children [Random_0], properties {delayMax=0.15, delayMin=0.15}}
 {id=Random_0, children [WavePlayer_0,WavePlayer_1,WavePlayer_2], properties {weights [1,1,1]}}
 {id=WavePlayer_0, properties {soundWaveAssetPtr=&/Char_ChicleVeil_Reactive_01}}
 {id=WavePlayer_1, properties {soundWaveAssetPtr=&/Char_ChicleVeil_Reactive_02}}
 {id=WavePlayer_2, properties {soundWaveAssetPtr=&/Char_ChicleVeil_Reactive_03}}
]
properties {
 attenuationSettings=&Foley_Attenuation_Basic_Att
 maxDistance=1536
 soundClassObject=&Fort_Player_Foley
 volumeMultiplier=2.75
}
